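# Sound cue for aquarium glass breaking, layering a shatter sound with a water splash.
# A mixer plays two randomized branches at once, one of five glass breaks and one of three splashes, with the splash layer mixed louder.
# Each branch passes through a modulator for variation, with the glass pitched down.
# It plays in the ambient props sound class using aquarium glass attenuation, with occlusion filtering enabled in the overrides.
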SoundCue /Hera/Environment/Blueprints/Aquarium/Audio/SC_AquariumGlass_Break
root=Mixer_0
nodes [
 {id=Mixer_0, children [Modulator_0,Modulator_1], properties {inputVolume [1,1.75]}}
 {id=Modulator_0, children [Random_0], properties {pitchMax=0.95, pitchMin=0.85}}
 {id=Modulator_1, children [Random_1]}
 {id=Random_0, children [WavePlayer_2,WavePlayer_3,WavePlayer_4,WavePlayer_5,WavePlayer_6], properties {weights [1,1,1,1,1]}}
 {id=Random_1, children [WavePlayer_7,WavePlayer_8,WavePlayer_9], properties {weights [1,1,1]}}
 {id=WavePlayer_2, properties {soundWaveAssetPtr=&/AquariumGlass_Break_05}}
 {id=WavePlayer_3, properties {soundWaveAssetPtr=&/AquariumGlass_Break_04}}
 {id=WavePlayer_4, properties {soundWaveAssetPtr=&/AquariumGlass_Break_03}}
 {id=WavePlayer_5, properties {soundWaveAssetPtr=&/AquariumGlass_Break_02}}
 {id=WavePlayer_6, properties {soundWaveAssetPtr=&/AquariumGlass_Break_01}}
 {id=WavePlayer_7, properties {soundWaveAssetPtr=&/AquariumGlass_Break_Splash_01}}
 {id=WavePlayer_8, properties {soundWaveAssetPtr=&/AquariumGlass_Break_Splash_02}}
 {id=WavePlayer_9, properties {soundWaveAssetPtr=&/AquariumGlass_Break_Splash_03}}
]
properties {
 attenuationSettings=&/ATT_AquariumGlass
 maxDistance=2750
 soundClassObject=&Fort_Ambient_Props_Athena
 volumeMultiplier=1.5
 attenuationOverrides {bEnableOcclusion=true, occlusionLowPassFilterFrequency=10000, occlusionTraceChannel=ECC_GameTraceChannel11, attenuationShapeExtents {x=0}}
}
