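# Sound cue that plays aquarium glass cracking sounds as ambient-class audio.
# A random node picks one of five equally weighted crack waves, then a modulator varies pitch from slightly lowered to normal before a mixer whose first input is empty.
# Spatialization uses the shared aquarium glass attenuation asset, and a concurrency override caps simultaneous instances at a few.
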SoundCue /Hera/Environment/Blueprints/Aquarium/Audio/SC_AquariumGlass_Crack
root=Mixer_0
nodes [
 {id=Mixer_0, children [null,Modulator_1], properties {inputVolume [1,1]}}
 {id=Modulator_1, children [Random_1], properties {pitchMax=1, pitchMin=0.8}}
 {id=Random_1, children [WavePlayer_6, WavePlayer_7, WavePlayer_8, WavePlayer_9, WavePlayer_10], properties {weights [1,1,1,1,1]}}
 {id=WavePlayer_10, properties {soundWaveAssetPtr=&/AquariumGlass_Crack_05}}
 {id=WavePlayer_6, properties {soundWaveAssetPtr=&/AquariumGlass_Crack_01}}
 {id=WavePlayer_7, properties {soundWaveAssetPtr=&/AquariumGlass_Crack_02}}
 {id=WavePlayer_8, properties {soundWaveAssetPtr=&/AquariumGlass_Crack_03}}
 {id=WavePlayer_9, properties {soundWaveAssetPtr=&/AquariumGlass_Crack_04}}
]
properties {
 attenuationSettings=&/ATT_AquariumGlass
 bOverrideConcurrency=true
 maxDistance=2750
 soundClassObject=&Fort_Ambient_Athena
 volumeMultiplier=1.5
 attenuationOverrides {bEnableOcclusion=true, occlusionLowPassFilterFrequency=10000, occlusionTraceChannel=ECC_GameTraceChannel11, attenuationShapeExtents {x=0}}
 concurrencyOverrides {maxCount=3}
}
