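# Sound cue for a long aquarium glass crack, classed as Fortnite ambient audio.
# A single crack wave feeds a Modulator that randomizes pitch slightly above normal on each play.
# The cue uses the shared aquarium glass attenuation asset and overrides concurrency to cap simultaneous instances.
# Its own occlusion attenuation values are stored but inactive, since attenuation override is not enabled.
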SoundCue /Hera/Environment/Blueprints/Aquarium/Audio/SC_AquariumGlass_Crack_Long
root=Modulator_1
nodes [
 {id=Modulator_1, children [WavePlayer_0], properties {pitchMax=1.15, pitchMin=1.05}}
 {id=WavePlayer_0, properties {soundWaveAssetPtr=&/AquariumGlass_Crack_Long_01}}
]
properties {
 attenuationSettings=&/ATT_AquariumGlass
 bOverrideConcurrency=true
 maxDistance=2750
 soundClassObject=&Fort_Ambient_Athena
 volumeMultiplier=1.2
 attenuationOverrides {bEnableOcclusion=true, occlusionLowPassFilterFrequency=10000, occlusionTraceChannel=ECC_GameTraceChannel11, attenuationShapeExtents {x=0}}
 concurrencyOverrides {maxCount=3}
}
