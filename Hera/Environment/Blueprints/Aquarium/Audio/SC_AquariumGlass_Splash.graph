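# Sound cue for an aquarium glass water splash, classed as Athena ambient sound and spatialized through the ATT_AquariumGlass attenuation asset.
# An equally weighted Random node picks one of five splash waves and feeds a fixed short Delay.
# The delayed sound passes through a default-setting Modulator and a single-input Mixer at the root, with the cue's overall volume boosted.
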SoundCue /Hera/Environment/Blueprints/Aquarium/Audio/SC_AquariumGlass_Splash
root=Mixer_0
nodes [
 {id=Delay_0, children [Random_0], properties {delayMax=0.3, delayMin=0.3}}
 {id=Mixer_0, children [Modulator_0], properties {inputVolume [1]}}
 {id=Modulator_0, children [Delay_0]}
 {id=Random_0, children [WavePlayer_1,WavePlayer_2,WavePlayer_3,WavePlayer_4,WavePlayer_5], properties {weights [1,1,1,1,1]}}
 {id=WavePlayer_1, properties {soundWaveAssetPtr=&/AquariumGlass_Water_Splash_01}}
 {id=WavePlayer_2, properties {soundWaveAssetPtr=&/AquariumGlass_Water_Splash_02}}
 {id=WavePlayer_3, properties {soundWaveAssetPtr=&/AquariumGlass_Water_Splash_03}}
 {id=WavePlayer_4, properties {soundWaveAssetPtr=&/AquariumGlass_Water_Splash_04}}
 {id=WavePlayer_5, properties {soundWaveAssetPtr=&/AquariumGlass_Water_Splash_05}}
]
properties {
 attenuationSettings=&/ATT_AquariumGlass
 maxDistance=2750
 soundClassObject=&Fort_Ambient_Athena
 volumeMultiplier=3
 attenuationOverrides {bEnableOcclusion=true, occlusionLowPassFilterFrequency=10000, occlusionTraceChannel=ECC_GameTraceChannel11, attenuationShapeExtents {x=0}}
}
