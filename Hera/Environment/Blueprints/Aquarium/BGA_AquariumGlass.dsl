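; Breakable aquarium glass actor that registers as an aim-assist target and cracks as it takes damage, or stays invulnerable when disabled.
; Server damage updates crack percent; a health threshold or death broadcasts AquariumBusted, makes it invulnerable and starts breaking.
; A timeline animates crack material with audio, then calls BreakGlass on the aquarium manager and destroys lights.
; RepNotifies drive cracks, health bar, glass visibility, collision and aim-assist removal.
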(event EventBeginPlay
  (AimAssist|RegisterAimAssistTarget 0 (Variables|Default|GetBox) false)
  (if (ScalableFloat|GetValueasBool (Variables|Default|GetEnabled))
    (Health|BindEventtoOnBuildingHealthChanged (delegate Custom|Damaged))
    (else
      (Variables|Actor|SetbCanBeDamaged false)
      (Variables|BuildingActor|SetbIsInvulnerable true))))

(event Custom|Damaged
  (Networking|SwitchHasAuthority
    (:Remote
      (Audio|Components|Audio|Play (Variables|Default|GetGlassHitAudio) 0.0))))

(event EventEndPlay
  (params (EndPlayReason :type EEndPlayReason%20Enum :default "Destroyed"))
  (AimAssist|UnregisterAimAssistTarget 0)
  (Development|PrintString "blah END PLAY"))

(event EventOnDamageServer
  (params (Damage :type Float%20%28single-precision%29) (DamageTags :type Gameplay%20Tag%20Container%20Structure%20%28by%20ref%29) (Momentum :type Vector :default "0, 0, 0") (HitInfo :type Hit%20Result%20Structure%20%28by%20ref%29) (InstigatedBy :type Controller%20Object%20Reference) (DamageCauser :type Actor%20Object%20Reference) (EffectContext :type Gameplay%20Effect%20Context%20Handle%20Structure))
  (bind _self self)
  (bind _returnvalue (Health|GetHealthPercent _self))
  (Variables|Default|SetLastDamageCauser InstigatedBy)
  (Variables|Default|SetGlassIsDamaged true)
  (Variables|Default|SetCrackPercent (- 1.0 _returnvalue))
  (if (<= _returnvalue (ScalableFloat|GetValueatLevel (Variables|Default|GetHealthPercentToBreakHF)))
    (Utilities|FlowControl|DoOnce false
      (:Completed
        (CustomVerb|BroadcastCustomVerbMessage (CustomVerb|MakeCustomVerbMessage (Variables|Default|GetLastDamageCauser) "(TagName=\"CustomVerb.AquariumBusted\")"))
        (CallFunction|StartBreakingSequence)
        (Variables|Actor|SetbCanBeDamaged false)
        (Variables|BuildingActor|SetbIsInvulnerable true)
        (Variables|Default|SetStartedBreaking true)))))

(event EventOnDeathServer
  (params (Damage :type Float%20%28single-precision%29) (DamageTags :type Gameplay%20Tag%20Container%20Structure%20%28by%20ref%29) (Momentum :type Vector :default "0, 0, 0") (HitInfo :type Hit%20Result%20Structure%20%28by%20ref%29) (InstigatedBy :type Controller%20Object%20Reference) (DamageCauser :type Actor%20Object%20Reference) (EffectContext :type Gameplay%20Effect%20Context%20Handle%20Structure))
  (Variables|Default|SetLastDamageCauser InstigatedBy)
  (Utilities|FlowControl|DoOnce false
    (:Completed
      (CustomVerb|BroadcastCustomVerbMessage (CustomVerb|MakeCustomVerbMessage (Variables|Default|GetLastDamageCauser) "(TagName=\"CustomVerb.AquariumBusted\")"))
      (CallFunction|StartBreakingSequence)
      (Variables|Actor|SetbCanBeDamaged false)
      (Variables|BuildingActor|SetbIsInvulnerable true)
      (Variables|Default|SetStartedBreaking true))))

(event Custom|TestWaterCollapse
  (Utilities|FlowControl|DoOnce false
    (:Completed
      (CustomVerb|BroadcastCustomVerbMessage (CustomVerb|MakeCustomVerbMessage (Variables|Default|GetLastDamageCauser) "(TagName=\"CustomVerb.AquariumBusted\")"))
      (CallFunction|StartBreakingSequence)
      (Variables|Actor|SetbCanBeDamaged false)
      (Variables|BuildingActor|SetbIsInvulnerable true)
      (Variables|Default|SetStartedBreaking true))))

(event Custom|StartBreakingSequence
  (Variables|Default|SetGlassIsDamaged true)
  (Utilities|FlowControl|Sequence
    (:then_0
      (bind (_direction _cracksize) (Timeline|Timeline PlayFromStart)
        (:Update
          (Rendering|Material|SetScalarParameterValueonMaterials (Variables|Default|GetSM_GlassCracks) "GlassCracking" (Math|Float|Lerp 0.5 0.0 _cracksize)))
        (:Finished
          (Variables|Default|SetGlassIsDestroyed true)
          (Class|BGAAquariumManager|BreakGlass (Variables|Default|GetAquariumManagerRef))
          (CallFunction|DisableLights))))
    (:then_1
      (Audio|Components|Audio|Play (Variables|Default|GetGlassCrackingAudio) 0.0))))

(fn DisableLights
  ; Add in whatever lighting components that are added and hook into DestroyComponent
  (Components|DestroyComponent (Variables|Default|GetBR_Pointlight_Component))
  (Components|DestroyComponent (Variables|Default|GetBR_Pointlight_Component))
  (Components|DestroyComponent (Variables|Default|GetLumenEmissiveLight_Component)))

(fn GetTargetOwner
  (return 0))

(fn IsValidAutoFireTarget
  (params (Attacker :type Actor%20Object%20Reference))
  (if (Variables|Default|GetStartedBreaking)
    (return false)
    (else
      (return true))))

(fn OnRep_CrackPercent
  (CallFunction|UpdateGlassCrackProgress))

(fn OnRep_GlassIsDamaged
  (if (Variables|Default|GetGlassIsDamaged)
    (CallFunction|SetSuppressHealthBar self true)
    (Rendering|SetVisibility (Variables|Default|GetSM_GlassCracks) true false)
    (else
      (CallFunction|SetSuppressHealthBar self false)
      (Rendering|SetVisibility (Variables|Default|GetSM_GlassCracks) false false))))

(fn OnRep_GlassIsDestroyed
  (if (Variables|Default|GetGlassIsDestroyed)
    (Rendering|SetVisibility (Variables|Default|GetPlane) false false)
    (Collision|SetCollisionEnabled (Variables|Default|GetBox) "NoCollision")
    (else
      (Rendering|SetVisibility (Variables|Default|GetPlane) true false)
      (Collision|SetCollisionEnabled (Variables|Default|GetBox) "QueryAndPhysics"))))

(fn OnRep_StartedBreaking
  (AimAssist|UnregisterAimAssistTarget 0))

(fn ShouldAlliesAutoFireOnDeployedTarget
  (return false))

(fn ShouldAutoFireOnSelfDeployedTarget
  (return false))

(fn UpdateGlassCrackProgress
  (Networking|SwitchHasAuthority
    (:Remote
      (if (not (Variables|Default|GetStartedBreaking))
        (Rendering|Material|SetScalarParameterValueonMaterials (Variables|Default|GetSM_GlassCracks) "GlassCracking" (Math|Float|Lerp 0.5 0.0 (Variables|Default|GetCrackPercent)))))))

(fn ConstructionScript)
