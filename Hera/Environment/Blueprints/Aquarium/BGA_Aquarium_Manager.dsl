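; Invulnerable aquarium manager actor that runs the one-time tank glass break.
; BreakGlass hides the pool via RepNotify, removes loot chests and fish, wets the walls, and plays splash FX and sounds.
; After a short delay, the authority destroys listed actors, water bodies and overlapping building pieces and heavily damages players and vehicles in the damage box.
; A RepNotify-shown puddle decal grows by timeline, and remote instances shake nearby players' cameras.
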(event EventBeginPlay
  ; Init
  (Variables|BuildingActor|SetbIsInvulnerable true)
  (Variables|Default|SetPuddleMeshEndScale (Transformation|GetWorldScale (Variables|Default|GetWetFloorDecal)))
  (bind _output_get (Variables|Default|SetPuddleMeshStartScale "0.695280,0.010000,0.010000"))
  (Transformation|SetWorldScale3D (Variables|Default|GetWetFloorDecal) _output_get))

(event Custom|BreakGlass
  ; Break the Glass
  (Utilities|FlowControl|DoOnce false
    (:Completed
      (Variables|Default|SetPoolVisible false)
      (for _array_element (Variables|Default|GetLootChests)
        (Utilities|IsValid _array_element
          (:Is%20Valid
            (Actor|DestroyActor _array_element))))
      ;; Huge water splash fx
      (Components|Activation|Activate (Variables|Default|GetNS_WaterTankBreak_HugeSplash) true)
      (Components|DestroyComponent (Variables|Default|GetNS_Env_TropicalFish_FishSchools_Tank))
      (Class|BPAquariumWetWalls|ApplyWetWalls (Variables|Default|GetWetWallsActorRef))
      (Audio|PlaySoundatLocation "/Hera/Environment/Blueprints/Aquarium/Audio/SC_AquariumGlass_Break.SC_AquariumGlass_Break" (Transformation|GetActorLocation))
      (Audio|Components|Audio|Play (Variables|Default|GetWaterSplashAudio) 0.0)
      (Utilities|FlowControl|Delay 0.7)
      (CallFunction|WaterDestroyActors)
      (Variables|Default|SetPuddleVisible true)
      (CallFunction|ApplyLocalCameraShake)
      (bind (_direction _puddle) (Timeline|PuddleMeshTimeline PlayFromStart)
        (:Update
          (Transformation|SetWorldScale3D (Variables|Default|GetWetFloorDecal) (Math|Vector|Lerp(Vector) (Variables|Default|GetPuddleMeshStartScale) (Variables|Default|GetPuddleMeshEndScale) _puddle)))))))

(event Custom|WaterDestroyActors
  ; water Damage Players and Props
  (Networking|SwitchHasAuthority
    (:Authority
      (Utilities|FlowControl|Sequence
        (:then_0
          (for _array_element (Variables|Default|GetBuildingActorToDestroyOnBreak)
            (Utilities|IsValid _array_element
              (:Is%20Valid
                (Actor|DestroyActor _array_element)))))
        (:then_1
          (for _array_element_2 (Variables|Default|GetWaterBodyActors)
            (Utilities|IsValid _array_element_2
              (:Is%20Valid
                (Actor|DestroyActor _array_element_2)))))
        (:then_2
          (for _array_element_3 (Collision|GetOverlappingActors (Variables|Default|GetPlayerDamageBox) "/Script/FortniteGame.FortPlayerPawnAthena" :!class /Script/Engine.PrimitiveComponent)
            (Environment|ApplyGlobalEnvironmentDamagetoActor _array_element_3 999.0)))
        (:then_3
          (for _array_element_4 (Collision|GetOverlappingActors (Variables|Default|GetPlayerDamageBox) "/Script/FortniteGame.FortAthenaVehicle" :!class /Script/Engine.PrimitiveComponent)
            (Environment|ApplyGlobalEnvironmentDamagetoActor _array_element_4 9999.0)))
        (:then_4
          (bind _outactors (Collision|BoxOverlapActorsWithOrientation (Transformation|GetWorldLocation (Variables|Default|GetPlayerDamageBox)) (Components|Box|GetScaledBoxExtent (Variables|Default|GetPlayerDamageBox)) (Transformation|GetWorldRotation (Variables|Default|GetPlayerDamageBox)) (Utilities|Array|MakeArray "ObjectTypeQuery1") 0 (Utilities|Array|MakeArray self)))
          (for _array_element_5 _outactors
            (Utilities|IsA(soft) _array_element_5 "/Script/FortniteGame.BuildingSMActor"
              (:True
                (Actor|DestroyActor _array_element_5)))))))))

(event Custom|ApplyLocalCameraShake
  ; Local Camera Shake
  (Networking|SwitchHasAuthority
    (:Remote
      (bind (_outactors _returnvalue) (Collision|SphereOverlapActors (- (Transformation|GetWorldLocation (Variables|Default|GetPlayerDamageBox)) "0.000000,0.000000,256.000000") 1.5e+03 (Utilities|Array|MakeArray "ObjectTypeQuery3") "/Script/FortniteGame.FortPlayerPawnAthena"))
      (if _returnvalue
        (for _array_element _outactors
          (bind _asfort_player_controller_athena (Utilities|Casting|CastToFortPlayerControllerAthena (Fortnite|GetControllerfromActor _array_element)))
          (Game|Feedback|ClientStartCameraShake :self _asfort_player_controller_athena :Shake "/Hera/Environment/Blueprints/Aquarium/B_AquariumCameraShake.B_AquariumCameraShake_C" :Scale 1.0 :PlaySpace "CameraLocal"))))))

(fn OnRep_PoolVisible
  (bind _poolref (Variables|Default|GetPoolRef))
  (bind _ns_env_tropicalfish_fishschools_tank (Variables|Default|GetNS_Env_TropicalFish_FishSchools_Tank))
  (if (not (Variables|Default|GetPoolVisible))
    (Collision|SetActorEnableCollision _poolref false)
    (Rendering|SetVisibility (Variables|/Script/Water.WaterBody|GetWaterBodyComponent _poolref) false false)
    (Utilities|IsValid _ns_env_tropicalfish_fishschools_tank
      (:Is%20Valid
        (Components|DestroyComponent _ns_env_tropicalfish_fishschools_tank)))))

(fn OnRep_PuddleVisible
  (if (Variables|Default|GetPuddleVisible)
    (Rendering|SetVisibility (Variables|Default|GetWetFloorDecal) true false)
    (else
      (Rendering|SetVisibility (Variables|Default|GetWetFloorDecal) false false))))

(fn ConstructionScript)
