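; BP_AquariumWetWalls is an aquarium environment actor that shows or hides a decal for wet walls.
; BeginPlay sets IsWetActorVisible to false, and the ApplyWetWalls custom event sets it to true.
; The variable's RepNotify, OnRep_IsWetActorVisible, applies the state to Decal1's visibility.
; The construction script sets the decal's visibility from ShowWetDebug for previewing.
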(event EventBeginPlay
  (Variables|Default|SetIsWetActorVisible false))

(event Custom|ApplyWetWalls
  (Variables|Default|SetIsWetActorVisible true))

(fn OnRep_IsWetActorVisible
  (if (Variables|Default|GetIsWetActorVisible)
    (Rendering|SetVisibility (Variables|Default|GetDecal1) true false)
    (else
      (Rendering|SetVisibility (Variables|Default|GetDecal1) false false))))

(fn ConstructionScript
  (Rendering|SetVisibility (Variables|Default|GetDecal1) (Variables|Default|GetShowWetDebug) false))
